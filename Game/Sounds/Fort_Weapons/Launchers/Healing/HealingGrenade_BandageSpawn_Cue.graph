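# Sound cue for a healing grenade's bandage spawn, layering a surface impact with the spawn sound.
# A mixer plays one of four equally weighted random impact waves alongside a single bandage spawn wave that follows after a short fixed delay and is mixed louder.
# Routed to the consumables sound class, it overrides attenuation and limits concurrency to one instance, stopping the oldest.
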SoundCue /Game/Sounds/Fort_Weapons/Launchers/Healing/HealingGrenade_BandageSpawn_Cue
root=Mixer_0
nodes [
 {id=Delay_1, children [WavePlayer_1], properties {delayMax=0.1, delayMin=0.1}}
 {id=Mixer_0, children [Random_0,Delay_1], properties {inputVolume [0.5,1.4]}}
 {id=Random_0, children [WavePlayer_2,WavePlayer_3,WavePlayer_4,WavePlayer_5], properties {weights [1,1,1,1]}}
 {id=WavePlayer_1, properties {soundWaveAssetPtr=&/HealingGrenade_BandageSpawn_01}}
 {id=WavePlayer_2, properties {soundWaveAssetPtr=&/HealingGrenade_ImpactSurface_01}}
 {id=WavePlayer_3, properties {soundWaveAssetPtr=&/HealingGrenade_ImpactSurface_02}}
 {id=WavePlayer_4, properties {soundWaveAssetPtr=&/HealingGrenade_ImpactSurface_03}}
 {id=WavePlayer_5, properties {soundWaveAssetPtr=&/HealingGrenade_ImpactSurface_04}}
]
properties {
 attenuationSettings=&Fort_Consumable_Use_Athena_Att
 bOverrideAttenuation=true
 bOverrideConcurrency=true
 maxDistance=2200
 soundClassObject=&Fort_Consumables
 volumeMultiplier=1
 attenuationOverrides {falloffDistance=1700, stereoSpread=800, attenuationShapeExtents {x=500}}
 concurrencyOverrides {maxCount=1, resolutionRule=StopOldest}
}
