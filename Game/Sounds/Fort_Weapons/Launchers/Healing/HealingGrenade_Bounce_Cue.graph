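# Sound cue for the healing grenade's bounce off surfaces, routed to the ranged weapon projectile effects sound class.
# A random node picks one of four equally weighted impact waves, and a modulator varies its pitch and slightly boosts its volume with variation.
# It uses a grenade bounce attenuation and overrides concurrency to cap simultaneous plays, refusing new instances once the limit is reached.
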SoundCue /Game/Sounds/Fort_Weapons/Launchers/Healing/HealingGrenade_Bounce_Cue
root=Modulator_5
nodes [
 {id=Modulator_5, children [Random_1], properties {pitchMax=1.1, pitchMin=0.9, volumeMax=1.25, volumeMin=1.15}}
 {id=Random_1, children [WavePlayer_7, WavePlayer_8, WavePlayer_9, WavePlayer_10], properties {weights [1,1,1,1]}}
 {id=WavePlayer_10, properties {soundWaveAssetPtr=&/HealingGrenade_ImpactSurface_03}}
 {id=WavePlayer_7, properties {soundWaveAssetPtr=&/HealingGrenade_ImpactSurface_04}}
 {id=WavePlayer_8, properties {soundWaveAssetPtr=&/HealingGrenade_ImpactSurface_01}}
 {id=WavePlayer_9, properties {soundWaveAssetPtr=&/HealingGrenade_ImpactSurface_02}}
]
properties {
 attenuationSettings=/Game/Sounds/Attenuation/Fort_Grenade_Bounce_Att
 bOverrideConcurrency=true
 maxDistance=5500
 pitchMultiplier=1.1
 soundClassObject=&Fort_WeaponRanged_Projectile_Effects
 volumeMultiplier=0.6
 concurrencyOverrides {maxCount=3, resolutionRule=PreventNew}
}
