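# Sound cue for a dropped healing grenade weapon in Fortnite.
# A random node picks between two equally weighted drop recordings, played at reduced volume and slightly lowered pitch.
# It uses the shared weapon drop attenuation and concurrency assets and is routed through the loot collect sound class.
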SoundCue /Game/Sounds/Fort_Weapons/Launchers/Healing/HealingGrenade_Drop_Cue
root=Random_0
nodes [
 {id=Random_0, children [WavePlayer_1,WavePlayer_3], properties {weights [1,1]}}
 {id=WavePlayer_1, properties {soundWaveAssetPtr=&/HealingGrenade_Drop_01}}
 {id=WavePlayer_3, properties {soundWaveAssetPtr=&/HealingGrenade_Drop_02}}
]
properties {
 attenuationSettings=&Fort_Weapon_Drop_Att
 maxDistance=3000
 pitchMultiplier=0.9
 soundClassObject=&Fort_Loot_Collect
 volumeMultiplier=0.25
 concurrencyOverrides {resolutionRule=PreventNew}
 concurrencySet [&Weapon_Drop_StopOldest]
}
