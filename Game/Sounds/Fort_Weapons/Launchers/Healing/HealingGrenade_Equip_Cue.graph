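# Sound cue for equipping the healing grenade, classed as weapon foley.
# A short fixed delay feeds a random node that picks one of three equip recordings with equal weight.
# The cue plays slightly pitched up at reduced volume, using the shared weapon reload attenuation asset.
# Its own concurrency override allows a single instance, stopping the farthest before refusing new ones.
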SoundCue /Game/Sounds/Fort_Weapons/Launchers/Healing/HealingGrenade_Equip_Cue
root=Delay_0
nodes [
 {id=Delay_0, children [Random_0], properties {delayMax=0.15, delayMin=0.15}}
 {id=Random_0, children [WavePlayer_1,WavePlayer_2,WavePlayer_3], properties {weights [1,1,1]}}
 {id=WavePlayer_1, properties {soundWaveAssetPtr=&/HealingGrenade_Equip_01}}
 {id=WavePlayer_2, properties {soundWaveAssetPtr=&/HealingGrenade_Equip_03}}
 {id=WavePlayer_3, properties {soundWaveAssetPtr=&/HealingGrenade_Equip_02}}
]
properties {
 attenuationSettings=&Fort_Weapon_Reload_Att
 bOverrideConcurrency=true
 maxDistance=2250
 pitchMultiplier=1.1
 soundClassObject=&Fort_WeaponFoley
 volumeMultiplier=0.401
 attenuationOverrides {falloffDistance=5000, attenuationShapeExtents {x=1000}}
 concurrencyOverrides {maxCount=1, resolutionRule=StopFarthestThenPreventNew}
}
