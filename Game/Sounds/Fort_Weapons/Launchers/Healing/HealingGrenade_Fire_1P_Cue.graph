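# First-person Sound Cue for firing the healing grenade launcher.
# A mixer layers one of five random fire waves, pitched up, with one of three random reload waves at lower volume after a short fixed delay.
# An outer modulator adds slight volume variation to the combined result.
# It uses first-person weapon sound class, projectile attenuation and gunfire concurrency settings.
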SoundCue /Game/Sounds/Fort_Weapons/Launchers/Healing/HealingGrenade_Fire_1P_Cue
root=Modulator_2
nodes [
 {id=Delay_0, children [Random_3], properties {delayMax=0.25, delayMin=0.25}}
 {id=Mixer_3, children [Modulator_1,Modulator_0], properties {inputVolume [1.1,0.55]}}
 {id=Modulator_0, children [Delay_0], properties {pitchMax=1, pitchMin=1}}
 {id=Modulator_1, children [Random_0], properties {pitchMax=1.1, pitchMin=1.1}}
 {id=Modulator_2, children [Mixer_3], properties {pitchMax=1, pitchMin=1, volumeMax=1.1, volumeMin=1}}
 {
  id=Random_0
  children [WavePlayer_2,WavePlayer_3,WavePlayer_5,WavePlayer_6,WavePlayer_7]
  properties {bShouldExcludeFromBranchCulling=true, weights [1,1,1,1,1]}
 }
 {id=Random_3, children [WavePlayer_9, WavePlayer_10, WavePlayer_11], properties {bShouldExcludeFromBranchCulling=true, weights [1,1,1]}}
 {id=WavePlayer_10, properties {soundWaveAssetPtr=&/HealingGrenade_Reload_14}}
 {id=WavePlayer_11, properties {soundWaveAssetPtr=&/HealingGrenade_Reload_15}}
 {id=WavePlayer_2, properties {soundWaveAssetPtr=&/HealingGrenade_Fire_01}}
 {id=WavePlayer_3, properties {soundWaveAssetPtr=&/HealingGrenade_Fire_02}}
 {id=WavePlayer_5, properties {soundWaveAssetPtr=&/HealingGrenade_Fire_03}}
 {id=WavePlayer_6, properties {soundWaveAssetPtr=&/HealingGrenade_Fire_04}}
 {id=WavePlayer_7, properties {soundWaveAssetPtr=&/HealingGrenade_Fire_05}}
 {id=WavePlayer_9, properties {soundWaveAssetPtr=&/HealingGrenade_Reload_13}}
]
properties {
 attenuationSettings=&Fort_Weapon_Projectile_FirstPerson_Athena_Att
 maxDistance=2097152
 soundClassObject=&Fort_WeaponRanged_Fire_1P_PSM
 volumeMultiplier=0.9
 attenuationOverrides {distanceAlgorithm=NaturalSound, falloffDistance=5000, attenuationShapeExtents {x=1000}}
 concurrencySet [&Weapon_Gunfire_FirstPerson_Conc]
}
