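# Sound cue for the firing sound of the healing grenade launcher, layered by listener distance.
# Close, medium and distant layers each pick a random wave, pass through volume or pitch modulators, then their own projectile attenuation asset.
# Close and medium share a third-person ranged-fire sound class and are mixed before joining the much quieter-weighted distant layer, which has its own class.
# Local and global gunfire concurrency limit playback, and the random nodes are excluded from branch culling.
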SoundCue /Game/Sounds/Fort_Weapons/Launchers/Healing/HealingGrenade_Fire_Cue
root=Mixer_0
nodes [
 {id=Attenuation_0, children [Modulator_2], properties {attenuationSettings=&Fort_Weapon_Projectile_Close_Athena_Att}}
 {
  id=Attenuation_1
  children [Modulator_0]
  properties {attenuationSettings=&Fort_Weapon_Projectile_Distant_Mono_Athena_Att}
 }
 {
  id=Attenuation_6
  children [Modulator_1]
  properties {attenuationSettings=&Fort_Weapon_Projectile_Medium_Mono_Athena_Att}
 }
 {id=Mixer_0, children [Mixer_1,SoundClass_1], properties {inputVolume [1,0.1]}}
 {id=Mixer_1, children [SoundClass_0,SoundClass_6], properties {inputVolume [0.7,0.4]}}
 {id=Modulator_0, children [Modulator_4], properties {volumeMax=2, volumeMin=2}}
 {id=Modulator_1, children [Modulator_3], properties {volumeMax=2.1, volumeMin=2}}
 {id=Modulator_2, children [Modulator_8], properties {pitchMax=1, pitchMin=1}}
 {id=Modulator_3, children [Random_4], properties {pitchMax=1, pitchMin=1, volumeMax=0.6, volumeMin=0.5}}
 {id=Modulator_4, children [Random_5], properties {pitchMax=1, pitchMin=1, volumeMax=0.6, volumeMin=0.5}}
 {id=Modulator_8, children [Random_3], properties {pitchMax=1.1, pitchMin=1.1}}
 {
  id=Random_3
  children [WavePlayer_5,WavePlayer_6,WavePlayer_7,WavePlayer_8,WavePlayer_9]
  properties {bShouldExcludeFromBranchCulling=true, weights [1,1,1,1,1]}
 }
 {
  id=Random_4
  children [WavePlayer_10, WavePlayer_18, WavePlayer_19, WavePlayer_20]
  properties {bShouldExcludeFromBranchCulling=true, weights [1,1,1,1]}
 }
 {
  id=Random_5
  children [WavePlayer_21, WavePlayer_22, WavePlayer_23, WavePlayer_24]
  properties {bShouldExcludeFromBranchCulling=true, weights [1,1,1,1]}
 }
 {
  id=SoundClass_0
  children [Attenuation_0]
  properties {soundClassOverride=&Fort_WeaponRanged_Fire_3P_PSM}
 }
 {
  id=SoundClass_1
  children [Attenuation_1]
  properties {soundClassOverride=&Fort_WeaponRanged_Fire_Distant}
 }
 {
  id=SoundClass_6
  children [Attenuation_6]
  properties {soundClassOverride=&Fort_WeaponRanged_Fire_3P_PSM}
 }
 {id=WavePlayer_10, properties {soundWaveAssetPtr=&/HealingGrenade_Fire_Medium_01}}
 {id=WavePlayer_18, properties {soundWaveAssetPtr=&/HealingGrenade_Fire_Medium_02}}
 {id=WavePlayer_19, properties {soundWaveAssetPtr=&/HealingGrenade_Fire_Medium_03}}
 {id=WavePlayer_20, properties {soundWaveAssetPtr=&/HealingGrenade_Fire_Medium_04}}
 {id=WavePlayer_21, properties {soundWaveAssetPtr=&/HealingGrenade_Fire_Distant_01}}
 {id=WavePlayer_22, properties {soundWaveAssetPtr=&/HealingGrenade_Fire_Distant_02}}
 {id=WavePlayer_23, properties {soundWaveAssetPtr=&/HealingGrenade_Fire_Distant_03}}
 {id=WavePlayer_24, properties {soundWaveAssetPtr=&/HealingGrenade_Fire_Distant_04}}
 {id=WavePlayer_5, properties {soundWaveAssetPtr=&/HealingGrenade_Fire_01}}
 {id=WavePlayer_6, properties {soundWaveAssetPtr=&/HealingGrenade_Fire_02}}
 {id=WavePlayer_7, properties {soundWaveAssetPtr=&/HealingGrenade_Fire_03}}
 {id=WavePlayer_8, properties {soundWaveAssetPtr=&/HealingGrenade_Fire_04}}
 {id=WavePlayer_9, properties {soundWaveAssetPtr=&/HealingGrenade_Fire_05}}
]
properties {
 maxDistance=27500
 volumeMultiplier=1
 attenuationOverrides {distanceAlgorithm=NaturalSound, falloffDistance=5000, attenuationShapeExtents {x=1000}}
 concurrencySet [&Weapon_Gunfire_Local_Conc, &Weapon_Gunfire_Global_Conc]
}
