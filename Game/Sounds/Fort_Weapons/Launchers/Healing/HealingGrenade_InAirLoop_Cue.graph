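# Sound cue for the healing grenade's looping in-air projectile sound.
# A looping wave player feeds an enveloper that fades volume in while pitch falls from a raised start over the first half second.
# A modulator randomizes pitch per play while holding volume steady.
# The cue uses shared launcher in-air attenuation and overrides concurrency to prevent new instances once the limit is reached.
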SoundCue /Game/Sounds/Fort_Weapons/Launchers/Healing/HealingGrenade_InAirLoop_Cue
root=Modulator_1
nodes [
 {
  id=Enveloper_2
  children [WavePlayer_1]
  properties {pitchCurve {editorCurveData {keys [<at 0 1.25>, {time=0.5}]}}, volumeCurve {editorCurveData {keys [<at 0 0>,<at 0.5 1>]}}}
 }
 {id=Modulator_1, children [Enveloper_2], properties {pitchMin=0.9, volumeMax=1, volumeMin=1}}
 {id=WavePlayer_1, properties {bLooping=true, soundWaveAssetPtr=&/HealingGrenade_InAirLoop_01}}
]
properties {
 attenuationSettings=/Game/Sounds/Attenuation/Fort_Weapon_Launcher_Proj_InAir_Default_Att
 bOverrideConcurrency=true
 maxDistance=12750
 soundClassObject=&Fort_WeaponRanged_Fire_3P_PSM
 volumeMultiplier=1
 attenuationOverrides {falloffDistance=5000, attenuationShapeExtents {x=1000}}
 concurrencyOverrides {resolutionRule=PreventNew}
}
