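# Sound cue for the healing grenade in water, layering a water impact with a continuous underwater fuse sound.
# A mixer blends a looping bubble wave, at reduced volume, with a slightly boosted, randomly chosen one-shot from three equally weighted grenade water impacts.
# It plays in the ranged weapon sound class with close-projectile attenuation, and its concurrency override prevents new instances once the limit is reached.
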SoundCue /Game/Sounds/Fort_Weapons/Launchers/Healing/HealingGrenade_InWaterLoop_Cue
root=Mixer_0
nodes [
 {id=Mixer_0, children [WavePlayer_0,Random_0], properties {inputVolume [0.5,1.1]}}
 {id=Random_0, children [WavePlayer_2,WavePlayer_3,WavePlayer_4], properties {weights [1,1,1]}}
 {
  id=WavePlayer_0
  properties {bLooping=true, soundWaveAssetPtr=/Game/Sounds/Fort_GamePlay_Sounds/Explosives/Grenade_Proj_Fuse_Water_Bubbles_Loop}
 }
 {id=WavePlayer_2, properties {soundWaveAssetPtr=/Game/Sounds/Fort_GamePlay_Sounds/Explosives/Grenade_Impact_Water_03}}
 {id=WavePlayer_3, properties {soundWaveAssetPtr=/Game/Sounds/Fort_GamePlay_Sounds/Explosives/Grenade_Impact_Water_01}}
 {id=WavePlayer_4, properties {soundWaveAssetPtr=/Game/Sounds/Fort_GamePlay_Sounds/Explosives/Grenade_Impact_Water_02}}
]
properties {
 attenuationSettings=/Game/Sounds/Attenuation/Fort_Weapon_Projectile_Close_Athena_Batarang_Att
 bOverrideConcurrency=true
 maxDistance=2600
 soundClassObject=&Fort_WeaponRanged
 volumeMultiplier=0.85
 attenuationOverrides {falloffDistance=5000, attenuationShapeExtents {x=1000}}
 concurrencyOverrides {resolutionRule=PreventNew}
}
