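# First-person reload sound cue for the healing grenade launcher.
# A quality-level switch feeds its first two slots from an equal-weight random pick of three reload waves, and its last two from a single wave.
# Both branches pass through modulators for slight volume variation, plus pitch variation on the single-wave branch, and mixers with empty second inputs.
# It uses first-person weapon-fire sound class, projectile attenuation and gunfire concurrency.
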SoundCue /Game/Sounds/Fort_Weapons/Launchers/Healing/HealingGrenade_Reload_1P_Cue
root=QualityLevel_0
nodes [
 {id=Mixer_1, children [Modulator_2,null], properties {inputVolume [0.9,0.7]}}
 {id=Mixer_10, children [Modulator_13,null], properties {inputVolume [0.9,0.7]}}
 {id=Mixer_3, children [null,Modulator_0], properties {inputVolume [1.1,0.2]}}
 {id=Mixer_4, children [null, WavePlayer_15], properties {inputVolume [1,0.2]}}
 {id=Modulator_0, children [Random_3], properties {pitchMax=1, pitchMin=1}}
 {id=Modulator_13, children [Mixer_4], properties {pitchMax=1.1, pitchMin=1, volumeMax=1.1, volumeMin=1}}
 {id=Modulator_2, children [Mixer_3], properties {pitchMax=1, pitchMin=1, volumeMax=1.1, volumeMin=1}}
 {id=QualityLevel_0, children [Mixer_1,Mixer_1,Mixer_10,Mixer_10]}
 {id=Random_3, children [WavePlayer_9, WavePlayer_10, WavePlayer_11], properties {weights [1,1,1]}}
 {id=WavePlayer_10, properties {soundWaveAssetPtr=&/HealingGrenade_Reload_14}}
 {id=WavePlayer_11, properties {soundWaveAssetPtr=&/HealingGrenade_Reload_15}}
 {id=WavePlayer_15, properties {soundWaveAssetPtr=&/HealingGrenade_Reload_13}}
 {id=WavePlayer_9, properties {soundWaveAssetPtr=&HealingGrenade_Reload_13}}
]
properties {
 attenuationSettings=&Fort_Weapon_Projectile_FirstPerson_Athena_Att
 maxDistance=2097152
 soundClassObject=&Fort_WeaponRanged_Fire_1P_PSM
 volumeMultiplier=1
 attenuationOverrides {distanceAlgorithm=NaturalSound, falloffDistance=5000, attenuationShapeExtents {x=1000}}
 concurrencySet [&Weapon_Gunfire_FirstPerson_Conc]
}
