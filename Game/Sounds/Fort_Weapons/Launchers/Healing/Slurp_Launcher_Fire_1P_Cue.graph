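# First-person fire sound cue for the Slurp healing launcher, mixing a main shot with a quieter mechanical layer.
# The shot layer picks randomly from four equally weighted fire recordings with slight pitch variation.
# The mech layer plays one sample after a short fixed delay with wider pitch variation.
# It uses first-person weapon attenuation, a ranged-fire sound class and first-person gunfire concurrency.
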SoundCue /Game/Sounds/Fort_Weapons/Launchers/Healing/Slurp_Launcher_Fire_1P_Cue
root=Mixer_3
nodes [
 {id=Delay_0, children [WavePlayer_2], properties {delayMax=0.2, delayMin=0.2}}
 {id=Mixer_3, children [Modulator_1,Modulator_0], properties {inputVolume [0.9,0.3]}}
 {id=Modulator_0, children [Delay_0], properties {pitchMax=1.2, pitchMin=0.9}}
 {id=Modulator_1, children [Random_1], properties {pitchMax=1.1, pitchMin=0.9}}
 {id=Random_1, children [WavePlayer_0, WavePlayer_1, WavePlayer_4, WavePlayer_12], properties {weights [1,1,1,1]}}
 {id=WavePlayer_0, properties {soundWaveAssetPtr=&/WPN_Launcher_Slurp_Fire_1P_01}}
 {id=WavePlayer_1, properties {soundWaveAssetPtr=&/WPN_Launcher_Slurp_Fire_1P_02}}
 {id=WavePlayer_12, properties {soundWaveAssetPtr=&/WPN_Launcher_Slurp_Fire_1P_04}}
 {id=WavePlayer_2, properties {soundWaveAssetPtr=&/WPN_Launcher_Slurp_Mech_01}}
 {id=WavePlayer_4, properties {soundWaveAssetPtr=&/WPN_Launcher_Slurp_Fire_1P_03}}
]
properties {
 attenuationSettings=&Fort_Weapon_Projectile_FirstPerson_Athena_Att
 maxDistance=2097152
 soundClassObject=&Fort_WeaponRanged_Fire_1P_PSM
 volumeMultiplier=1
 attenuationOverrides {distanceAlgorithm=NaturalSound, falloffDistance=5000, attenuationShapeExtents {x=1000}}
 concurrencySet [&Weapon_Gunfire_FirstPerson_Conc]
}
