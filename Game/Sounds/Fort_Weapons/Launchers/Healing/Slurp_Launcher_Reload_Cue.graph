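# Sound cue for the reload sound of the healing Slurp Launcher weapon.
# A single reload wave player feeds a modulator at the root, with no variation between samples.
# It plays at reduced volume in the weapon foley class with the shared reload attenuation asset.
# Its own concurrency override allows one instance, stopping the farthest before preventing new ones.
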SoundCue /Game/Sounds/Fort_Weapons/Launchers/Healing/Slurp_Launcher_Reload_Cue
root=Modulator_0
nodes [
 {id=Modulator_0, children [WavePlayer_0]}
 {id=WavePlayer_0, properties {soundWaveAssetPtr=&/WPN_Launcher_Slurp_Reload_01}}
]
properties {
 attenuationSettings=&Fort_Weapon_Reload_Att
 bOverrideConcurrency=true
 maxDistance=2250
 soundClassObject=&Fort_WeaponFoley
 volumeMultiplier=0.501
 attenuationOverrides {falloffDistance=5000, attenuationShapeExtents {x=1000}}
 concurrencyOverrides {maxCount=1, resolutionRule=StopFarthestThenPreventNew}
}
